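# Sound cue for a looping positional ambient in the secret facility hydroponics area, under Greenhouse props.
# A single looping wave player feeds the root directly, playing the B_Loop_01 wave despite the cue's A_Loop name.
# It references shared attenuation and spot-ambience concurrency assets, and also stores capsule, HRTF, low-pass and focus attenuation values.
# A distance-scaled linear submix send routes it to the default reverb.
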SoundCue /Hera/Audio/Props/Greenhouse/SC_Spot_SecretFacility_Hydroponics_A_Loop
root=WavePlayer_1
nodes [{id=WavePlayer_1, properties {bLooping=true, soundWaveAssetPtr=&/Spot_SecretFacility_Hydroponics_B_Loop_01}}]
properties {
 attenuationSettings=&/ATT_Spot_SecretFacility_Hydroponics
 maxDistance=1300
 soundClassObject=&Fort_Ambient_OneShots_Outdoors
 volumeMultiplier=0.6
 attenuationOverrides {
  attenuationShape=Capsule
  bAttenuateWithLPF=true
  bEnableFocusInterpolation=true
  bEnableListenerFocus=true
  bEnableLogFrequencyScaling=true
  distanceAlgorithm=Custom
  focusReleaseInterpSpeed=0.5
  lPFFrequencyAtMax=600
  lPFRadiusMax=3600
  lPFRadiusMin=2000
  nonFocusVolumeAttenuation=0.75
  spatializationAlgorithm=SPATIALIZATION_HRTF
  stereoSpread=0
  attenuationShapeExtents {x=1000}
  customAttenuationCurve {editorCurveData {keys [
     <at 0 1>
     <at 0.05 1>
     {arriveTangent=-1.0526, interpMode=RCIM_Cubic, leaveTangent=-1.0526, time=0.5, value=0.33}
     {arriveTangent=-1.3611, interpMode=RCIM_Cubic, leaveTangent=-1.3611, tangentMode=RCTM_User, time=1, value=0}
    ]}}
 }
 concurrencyOverrides {maxCount=4}
 concurrencySet [/AmbientAudioBase/Concurrency/AmbientConc_3Max_Spot_Default_Concurrency]
 soundSubmixSends [{
   maxSendDistance=3600
   maxSendLevel=0.6
   minSendDistance=1500
   minSendLevel=0.15
   sendLevelControlMethod=Linear
   soundSubmix=&Fortnite_Reverb_Default
  }]
}
